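# Niagara system named as an additive, camera-facing mesh corona flare example in the NeverCook FX reuse folder, though it contains no such effect.
# Its only emitter, Empty, has no renderers, so the system draws nothing as it stands.
# System State is the only module used, with no user variables, data interfaces or dynamic inputs.
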NiagaraSystem /Game/NeverCook/FX/Niagara_FX_Reuse_Examples/Flare/NS_Corona_Flare_MeshCamFacing_Additive
dependencies {usedDataInterfaces [], usedDynamicInputs [], usedModules [&SystemState]}
emitters {Empty {rendererClasses [], renderers []}}
summary {systemName=NS_Corona_Flare_MeshCamFacing_Additive, emitters [{emitterName=Empty, rendererClasses []}], userVariables []}
