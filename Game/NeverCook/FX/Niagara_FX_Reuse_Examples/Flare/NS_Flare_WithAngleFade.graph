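# Niagara example system in the FX reuse examples folder, named as a flare glow with angle fade.
# It holds a single emitter, SignGlow, and uses SystemState as its only listed module.
# No renderers, user variables, dynamic inputs or data interfaces are listed, so the angle-fade logic is not visible here.
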NiagaraSystem /Game/NeverCook/FX/Niagara_FX_Reuse_Examples/Flare/NS_Flare_WithAngleFade
dependencies {usedDataInterfaces [], usedDynamicInputs [], usedModules [&SystemState]}
emitters {SignGlow {rendererClasses [], renderers []}}
summary {systemName=NS_Flare_WithAngleFade, emitters [{emitterName=SignGlow, rendererClasses []}], userVariables []}
